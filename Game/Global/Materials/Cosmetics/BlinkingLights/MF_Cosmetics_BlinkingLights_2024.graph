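# Material function that adds blinking lights to cosmetic materials, overriding base color, emissive, metallic, roughness and specular behind a "Use Lights" static switch.
# A selectable mask channel marks the lights, its raw value apparently serving as a per-light phase offset.
# Smoothstepped sine waves blend synchronized and random blinking, with blink and on/off controls.
# Fresnel shapes the glow and unlit tint, while a time-of-day function scales evening and night emissive.
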MaterialFunction /Game/Global/Materials/Cosmetics/BlinkingLights/MF_Cosmetics_BlinkingLights_2024
graph {nodes [
  n1 = SetMaterialAttributes!(
   MaterialAttributes=NamedRerouteUsage()
   BaseColor=NamedRerouteUsage()
   EmissiveColor=NamedRerouteUsage()
   Metallic=NamedRerouteUsage()
   Roughness=NamedRerouteUsage()
   Specular=NamedRerouteUsage()
  ) {attributeSetTypes [
    #1
    #2
    #3
    #4
    #5
   ]}
  FunctionOutput(None=StaticSwitchParameter(True=@n1, False=NamedRerouteUsage()) {group=Lights, parameterName=Use Lights, sortPriority=0})
  n2 = ChannelMaskParameter(None=FunctionInput() {bUsePreviewValueAsDefault=true, inputName=Mask, sortPriority=10, previewValue=<1 0 0 1>}) {group=Lights, maskChannel=Blue, parameterName=LightMask, sortPriority=1, defaultValue=<rgba 0 0 1 0>}
  n3 = NamedRerouteDeclaration(None=@n2.R) {name=LightOffsetMask, nodeColor=<rgba 1 0.2578 0 1>}
  NamedRerouteDeclaration(None=Saturate(None=Ceil(None=@n3))) {name=LightsMask, nodeColor=<rgba 0.7812 0 1 1>}
  n4 = GetMaterialAttributes(None=FunctionInput() {inputType=FunctionInput_MaterialAttributes}) {attributeGetTypes [
    #1
    #2
    #4
    #5
    #3
   ]}
  n5 = Multiply(A=@n4.Specular, B=ScalarParameter() {defaultValue=1, group=Lights, parameterName=SpecularMultiplier})
  n6 = NamedRerouteUsage()
  NamedRerouteDeclaration(None=LinearInterpolate(A=@n4.Specular, B=Saturate(None=@n5), Alpha=@n6)) {name=InSpecular, nodeColor=<rgba 0 0.8125 1 1>}
  NamedRerouteDeclaration(None=LinearInterpolate(A=@n4.Metallic, B=ScalarParameter() {defaultValue=0.75, group=Lights, parameterName=Metalness, sliderMax=1}, Alpha=@n6)) {name=InMetallic, nodeColor=<rgba 0.5234 1 0 1>}
  NamedRerouteDeclaration(None=@n4.EmissiveColor) {name=InEmissive, nodeColor=<rgba 0.1953 0 1 1>}
  NamedRerouteDeclaration(None=@n4.BaseColor) {name=InBaseColor, nodeColor=<rgba 0.1484 1 0 1>}
  NamedRerouteDeclaration(None=@n4.MaterialAttributes) {name=InMaterialAttributes, nodeColor=<rgba 0 1 0.5781 1>}
  n7 = NamedRerouteUsage()
  n8 = Multiply(A=NamedRerouteUsage(), B=ScalarParameter() {defaultValue=0.5, group=Lights, parameterName=LightOffColorIntensity})
  n9 = LinearInterpolate(A=@n8, B=Multiply(A=@n8) {constB=5}, Alpha=Fresnel() {baseReflectFraction=0, exponent=0.5})
  n10 = NamedRerouteUsage()
  NamedRerouteDeclaration(None=LinearInterpolate(A=LinearInterpolate(A=@n7, B=@n9, Alpha=@n10), Alpha=NamedRerouteUsage()) {constB=0}) {name=OutBaseColor, nodeColor=<rgba 1 0.2578 0 1>}
  n11 = ScalarParameter() {defaultValue=0.3, group=Lights, parameterName=BlinkSpeed}
  n12 = Multiply(A=Time(), B=@n11)
  n13 = Saturate(None=Multiply(A=Sine(None=@n12)) {constB=2})
  n14 = Time() {bOverride_Period=true, period=20}
  n15 = NamedRerouteUsage()
  n16 = Add(A=Time(), B=Add(A=Multiply(A=@n14, B=@n15), B=@n15))
  n17 = Saturate(None=Sine(None=Multiply(A=@n11, B=@n16)))
  n18 = ScalarParameter() {defaultValue=1, group=Lights, parameterName=LightRandomBlink, sliderMax=1}
  n19 = LinearInterpolate(
   A=Constant() {r=1}
   B=SmoothStep(Value=LinearInterpolate(A=@n13, B=@n17, Alpha=@n18)) {constMax=0.5}
   Alpha=NamedRerouteDeclaration(None=ScalarParameter() {defaultValue=1, group=Lights, parameterName=EnableLightBlink, sliderMax=1}) {name=EnableLightBlink, nodeColor=<rgba 0 1 0.1094 1>}
  )
  n20 = NamedRerouteDeclaration(None=ScalarParameter() {defaultValue=1, group=Lights, parameterName=LightsOn, sliderMax=1}) {name=LightsOn, nodeColor=<rgba 0.4297 0 1 1>}
  NamedRerouteDeclaration(None=Multiply(A=NamedRerouteUsage(), B=Multiply(A=@n19, B=@n20))) {name=AnimatedMask, nodeColor=<rgba 0 0.0625 1 1>}
  n21 = NamedRerouteUsage()
  n22 = ScalarParameter() {defaultValue=0.1, group=Lights, parameterName=FresnelExponent}
  n23 = Saturate(None=Power(
    Base=Fresnel(ExponentIn=@n22) {baseReflectFraction=0, exponent=0.1}
    Exp=ScalarParameter() {defaultValue=0.1, group=Lights, parameterName=FresnelPower}
   ))
  n24 = Desaturation(
   None=Multiply(A=LinearInterpolate(A=@n21, Alpha=@n23) {constB=0}, B=ScalarParameter() {defaultValue=300, group=Lights, parameterName=LightColorIntensity})
   Fraction=Multiply(A=ScalarParameter() {defaultValue=1, group=Lights, parameterName=LightColorSaturation}) {constB=-1}
  )
  n25 = Multiply(A=@n24, B=ScalarParameter() {defaultValue=0.75, parameterName=NightTimeIntensityMultiplier}) {constB=0.75}
  NamedRerouteDeclaration(None=Multiply(
    A=MaterialFunctionCall(Color_Morning=@n24, Color_Day=@n24, Color_Evening=@n25, Color_Night=@n25) {materialFunction=&Color_Time_of_Day}.Result
    B=NamedRerouteUsage()
   )) {name=OutEmissive, nodeColor=<rgba 0.3828 1 0 1>}
  n26 = Multiply(A=@n4.Roughness, B=ScalarParameter() {defaultValue=1, group=Lights, parameterName=RoughnessMultiplier})
  NamedRerouteDeclaration(None=LinearInterpolate(A=@n4.Roughness, B=Saturate(None=@n26), Alpha=@n6)) {name=InRoughness, nodeColor=<rgba 1 0 0.4922 1>}
 ]}
