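# Masked, two-sided parent material for the lights on the AlmondSplash gadget cosmetic, drawing them as round billboards with sphere-like normals.
# It layers the 2024 blinking-lights function between the 2023 character parent's PreFX and PostFX functions.
# Centered UV1 coordinates drive world position offset for view-facing quads, circular opacity masking and derived normals.
# The billboards are pushed toward the camera except in the shadow pass, and a parameter previews their collapsed state.
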Material /BRCosmetics/Animation/Props/Gadgets_AlmondSplash/Materials/M_AlmondSplash_Lights_Parent
graph {
 properties {
  bUseMaterialAttributes=true
  bUsedWithMeshDeformer=true
  bUsedWithMorphTargets=true
  bUsedWithSkeletalMesh=true
  blendMode=BLEND_Masked
  pixelDepthOffsetMode=PDOM_Legacy
  shadingModel=MSM_FromMaterialExpression
  twoSided=true
  usedShadingModels=Default Lit | Subsurface | Clear Coat | Subsurface Profile | Cloth
 }
 nodes [
  n1 = MaterialFunctionCall() {materialFunction=&MF_CharacterParent_2023_PreFX}
  MaterialFunctionCall(
   Material=MaterialFunctionCall!(In=NamedRerouteUsage(), Mask=@n1.Tech Art Mask) {materialFunction=/Game/Global/Materials/Cosmetics/BlinkingLights/MF_Cosmetics_BlinkingLights_2024}.Result
   M=@n1.M
   Pre Skinned Local Position=@n1.M
   Distance Blend=@n1.M
   Fresnel=@n1.M
  ) {materialFunction=&MF_CharacterParent_2023_PostFX}
  NamedRerouteDeclaration(None=@n1.Result) {name=InMaterialAttributes, nodeColor=<rgba 1 0 0.7266 1>}
  n2 = Multiply(A=NamedRerouteUsage()) {constB=2}
  NamedRerouteDeclaration(None=Normalize(VectorInput=DeriveNormalZ(InXY=@n2))) {name=SphereNormals, nodeColor=<rgba 1 0.7969 0 1>}
  NamedRerouteDeclaration(None=SphereMask(A=@n2, B=Constant()) {attenuationRadius=1}) {name=OpacityMask, nodeColor=<rgba 0.4766 0 1 1>}
  n3 = TextureCoordinate() {coordinateIndex=1}
  n4 = AppendVector(A=NamedRerouteDeclaration(None=Subtract(A=@n3) {constB=0.5}) {name=UVs, nodeColor=<rgba 0 1 0.9062 1>}, B=Constant())
  n5 = Multiply(
   A=ScalarParameter() {defaultValue=-2.5, group=Billboard Lights, parameterName=CollapsedVertScale}
   B=MaterialFunctionCall() {materialFunction=&ObjectScale}.Scale XYZ
  )
  n6 = Multiply(A=Transform(None=@n4), B=@n5) {constB=-1}
  n7 = Constant3Vector() {constant=<rgba 1 -1 1 0>}
  n8 = Multiply(A=OneMinus(None=@n5), B=Transform(None=Multiply(A=@n4, B=@n7)) {transformSourceType=TRANSFORMSOURCE_View})
  n9 = ScalarParameter() {defaultValue=1, group=Billboard Lights, parameterName=BillboardScale}
  n10 = LinearInterpolate(A=Add(A=@n6, B=Multiply(A=@n8, B=@n9)), B=@n6, Alpha=ScalarParameter() {group=Billboard Lights, parameterName=PreviewCollapsedBillboard})
  n11 = CameraPositionWS()
  n12 = WorldPosition()
  n13 = Multiply(
   A=Normalize(VectorInput=Subtract(A=@n11, B=@n12.XYZ))
   B=ScalarParameter() {defaultValue=3, group=Billboard Lights, parameterName=PushTowardsCamera}
  )
  NamedRerouteDeclaration(None=ShadowReplace(Default=Add(A=@n10, B=@n13), Shadow=@n10)) {name=BillboardWPO, nodeColor=<rgba 0 1 0.1562 1>}
  NamedRerouteDeclaration(None=SetMaterialAttributes!(
    MaterialAttributes=NamedRerouteUsage()
    Opacity Mask=NamedRerouteUsage()
    Normal=NamedRerouteUsage()
    World Position Offset=NamedRerouteUsage()
   ) {attributeSetTypes [#1, #2, #3]}) {name=OutMaterialAttributes, nodeColor=<rgba 1 0 0.3281 1>}
 ]
}
